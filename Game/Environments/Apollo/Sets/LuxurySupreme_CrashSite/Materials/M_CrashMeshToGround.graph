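# Nanite-ready material for Luxury Supreme crash-site wreckage that blends meshes into the Apollo landscape.
# A tinted base with optional textures and detail normals gains metallic damage from world-aligned noise, slope, height and painted masks.
# At Epic quality on deferred, landscape RVT color and normals fade in by height above terrain, with a NoRVT switch.
# An Emissive switch substitutes a flickering glow branch, and proxy replacements simplify baked HLOD output.
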Material /Game/Environments/Apollo/Sets/LuxurySupreme_CrashSite/Materials/M_CrashMeshToGround
graph {
 properties {
  bTangentSpaceNormal=false
  bUseMaterialAttributes=true
  bUsedWithNanite=true
  pixelDepthOffsetMode=PDOM_Legacy
  usedShadingModels=Default Lit
 }
 nodes [
  n1 = NamedRerouteUsage()
  n2 = NamedRerouteUsage()
  n3 = NamedRerouteUsage()
  n4 = SetMaterialAttributes!(MaterialAttributes=@n3, Base Color=Constant() {r=0.25}, Metallic=Constant() {r=1}, Roughness=Constant() {r=0.4}, Custom Data 0=Constant()) {attributeSetTypes [
    #1
    #2
    #3
    #4
   ]}
  n5 = GetMaterialAttributes() {attributeGetTypes [#1]}
  n6 = GetMaterialAttributes(None=NamedRerouteUsage()) {attributeGetTypes [#3]}
  n7 = Saturate(None=Divide(A=NamedRerouteUsage(), B=Constant() {r=2}))
  n8 = LinearInterpolate(A=Constant() {desc=0.1, r=0.11}, B=Constant() {desc=0.52, r=0.55}, Alpha=@n7)
  n9 = Constant() {r=0.38}
  n10 = ShadingPathSwitch(Default=@n8, Deferred=@n8, Forward=@n9, Mobile=@n9)
  n11 = MaterialFunctionCall(Transition Phase=@n10, Height Texture=NamedRerouteUsage(), Contrast=Constant() {r=5}) {materialFunction=&HeightLerp}
  n12 = LinearInterpolate(A=Constant() {r=0.25}, B=Constant() {r=0.35}, Alpha=@n11.Alpha)
  n13 = NamedRerouteUsage()
  n14 = Constant() {r=2}
  n15 = LinearInterpolate(A=Constant() {r=0.2}, B=@n12, Alpha=Saturate(None=Divide(A=@n13, B=@n14)))
  n16 = LinearInterpolate(A=Constant() {r=0.5}, B=Constant() {r=0.95}, Alpha=@n6.Roughness)
  n17 = Constant() {r=0.75}
  n18 = SetMaterialAttributes!(
   MaterialAttributes=@n6.MaterialAttributes
   Custom Data 0=Constant()
   Base Color=ShadingPathSwitch(Default=@n15, Deferred=@n15, Forward=@n12, Mobile=@n12)
   Roughness=LinearInterpolate(A=Constant() {r=0.42}, B=Multiply(A=@n16, B=@n17), Alpha=@n11.Alpha)
   Metallic=Constant() {r=1}
  ) {attributeSetTypes [
    #4
    #1
    #3
    #2
   ]}
  n19 = NamedRerouteUsage()
  n20 = ComponentMask(None=@n19) {g=true, r=true}
  n21 = LinearInterpolate(A=Constant() {r=5}, Alpha=@n7)
  n22 = Constant() {r=4}
  n23 = ShadingPathSwitch(Default=@n21, Deferred=@n21, Forward=@n22, Mobile=@n22)
  n24 = MaterialFunctionCall(
   A=AppendVector(A=Multiply(A=@n20, B=@n23), B=ComponentMask(None=@n19) {b=true})
   B=Constant3Vector() {constant=<rgba 0 0 1 0>}
   Transition Phase=@n10
   Height Texture=NamedRerouteUsage()
   Contrast=Constant() {r=2}
  ) {materialFunction=&HeightLerp}
  n25 = ScalarParameter() {defaultValue=-0.35, group=Damage, parameterName=AngleMin}
  n26 = ScalarParameter() {defaultValue=1, group=Damage, parameterName=Angle Max}
  n27 = ComponentMask(None=VertexNormalWS()) {b=true}
  n28 = LinearInterpolate(A=Constant() {r=0.15}, B=Constant() {r=0.75}, Alpha=Saturate(None=LinearInterpolate(A=@n25, B=@n26, Alpha=@n27)))
  n29 = MaterialFunctionCall(
   Transition Phase=@n28
   Height Texture=NamedRerouteDeclaration(None=@n24.Lerp Alpha No Contrast) {name=Damage, nodeColor=<rgba 0 0.8359 1 1>}
   Contrast=Constant() {r=3}
  ) {materialFunction=&HeightLerp}
  n30 = TextureSampleParameter2D(UVs=TextureCoordinate() {coordinateIndex=1}) {
   group=Damage
   parameterName=DamageMask
   samplerType=SAMPLERTYPE_LinearGrayscale
   sortPriority=4
   texture=/Game/Environments/Apollo/Sets/LuxurySupreme_CrashSite/Textures/T_LS_Crash_Wreckage_Mask_01_M
  }
  n31 = NamedRerouteUsage()
  n32 = MaterialFunctionCall(
   Transition Phase=Multiply(A=@n30.R, B=GetMaterialAttributes(None=@n31) {attributeGetTypes [#3]}.Roughness)
   Height Texture=@n30.R
   Contrast=ScalarParameter() {defaultValue=2, parameterName=Damage Mask Contrast}
  ) {materialFunction=&HeightLerp}
  n33 = StaticSwitchParameter(True=Multiply(A=@n29.Alpha, B=OneMinus(None=@n32.Alpha)), False=@n29.Alpha) {group=Damage, parameterName=PaintedDamage}
  n34 = QualitySwitch(Default=@n29.Alpha, Low=@n29.Alpha, High=@n33, Medium=@n29.Alpha, Epic=@n33)
  n35 = BlendMaterialAttributes(
   A=SetMaterialAttributes!(MaterialAttributes=@n18, Normal=NamedRerouteDeclaration(None=@n24.Results) {name=Damage Normals, nodeColor=<rgba 1 0.6328 0 1>}) {attributeSetTypes [#5]}
   B=NamedRerouteUsage()
   Alpha=ShadingPathSwitch(Default=@n34, Deferred=@n34, Forward=@n29.Alpha, Mobile=@n29.Alpha)
  ) {vertexAttributeBlendType=UseA}
  n36 = Multiply(A=NamedRerouteUsage(), B=Constant() {r=0.5})
  n37 = NamedRerouteUsage()
  n38 = BlendMaterialAttributes(
   A=SetMaterialAttributes!(
    MaterialAttributes=@n35
    Base Color=NamedRerouteDeclaration(None=@n5.Base Color) {name=Corruption, nodeColor=<rgba 0.1953 1 0 1>}
    Metallic=Constant() {r=1}
   ) {attributeSetTypes [#1, #2]}
   B=@n35
   Alpha=Saturate(None=Add(A=@n36, B=@n37))
  )
  n39 = StaticSwitchParameter(
   True=BlendMaterialAttributes(A=@n3, B=@n4, Alpha=NamedRerouteUsage())
   False=ShadingPathSwitch(Default=@n38, Deferred=@n38, Forward=@n35, Mobile=@n35)
  ) {group=RVT, parameterName=NoRVT}
  n40 = NamedRerouteDeclaration(None=MaterialFunctionCall!(Material=StaticSwitchParameter(True=@n2, False=@n39) {parameterName=Emissive}) {materialFunction=/Game/Athena/Apollo/Maps/Landscape/VT/MaterialFunctions/TransformTangentToWorld}.Result) {name=BeforeRVT, nodeColor=<rgba 1 0 0.1172 1>}
  n41 = NamedRerouteUsage()
  n42 = RuntimeVirtualTextureSampleParameter() {materialType=BaseColor_Normal_Specular_YCoCg, parameterName=RVT, virtualTexture=/Game/Athena/Apollo/Maps/Landscape/VT/RVT_Apollo_Landscape}
  n43 = Add(A=WorldPosition().XYZ, B=OneMinus(None=ScalarParameter() {group=RVT, parameterName=RVTOffset}))
  n44 = Constant()
  n45 = RuntimeVirtualTextureSampleParameter() {
   group=RVT
   materialType=WorldHeight
   parameterName=RVTHeight
   virtualTexture=/Game/Athena/Apollo/Maps/Landscape/VT/RVT_Apollo_Landscape_Height
  }
  n46 = NamedRerouteDeclaration(None=Subtract(
    A=ComponentMask(None=@n43) {b=true}
    B=MaterialFunctionCall!(NoRVT=@n44, RVT=@n45.WorldHeight) {materialFunction=/Game/Packages/Fortress_SharedMaterials/Base_Material_Functions/FN_RVT_Switch_Color}.Result
   )) {name=RVTHeight Raw, nodeColor=<rgba 0 0.4609 1 1>}
  n47 = VertexNormalWS()
  n48 = Divide(
   A=StaticSwitchParameter(True=ScalarParameter() {defaultValue=500, group=RVT, parameterName=RVTFade}, False=NamedRerouteUsage()) {parameterName=Emissive}
   B=LinearInterpolate(A=Constant() {r=12}, Alpha=ComponentMask(None=@n47) {b=true}) {constA=12}
  ) {constB=8}
  n49 = StaticSwitchParameter(True=@n48, False=ScalarParameter() {defaultValue=25, group=RVT, parameterName=RVT Fade Normals}) {parameterName=Emissive}
  n50 = BlendMaterialAttributes(
   A=SetMaterialAttributes!(MaterialAttributes=@n41, Normal=@n42.Normal) {attributeSetTypes [#5]}
   B=@n41
   Alpha=Saturate(None=Divide(A=@n46, B=@n49))
  ) {vertexAttributeBlendType=UseA}
  n51 = Constant()
  n52 = SetMaterialAttributes!(
   MaterialAttributes=@n50
   Base Color=@n42.BaseColor
   Specular=@n42.BaseColor
   Roughness=@n42.BaseColor
   Metallic=@n51
   Custom Data 0=@n51
   Emissive Color=@n51
  ) {attributeSetTypes [
    #1
    #6
    #3
    #2
    #4
    #7
   ]}
  n53 = StaticSwitchParameter(True=@n48, False=ScalarParameter() {defaultValue=255, group=RVT, parameterName=RVT Fade Color}) {parameterName=Emissive}
  n54 = BlendMaterialAttributes(A=@n52, B=@n50, Alpha=Saturate(None=Divide(A=@n46, B=@n53))) {vertexAttributeBlendType=UseA}
  n55 = QualitySwitch(
   Default=@n1
   Low=@n1
   High=@n1
   Medium=@n1
   Epic=MaterialFunctionCall(NoRVT=@n40, RVT=MaterialProxyReplace(Realtime=@n54, MaterialProxy=@n40)) {materialFunction=&FN_RVT_Switch_Material}.Result
  )
  n56 = StaticSwitchParameter(True=@n1, False=ShadingPathSwitch(Default=@n55, Deferred=@n55, Forward=@n1, Mobile=@n1)) {group=RVT, parameterName=NoRVT}
  n57 = GetMaterialAttributes(None=@n56) {attributeGetTypes [#5, #7]}
  n58 = NamedRerouteUsage()
  n59 = Constant()
  MaterialProxyReplace(
   Realtime=@n56
   MaterialProxy=SetMaterialAttributes!(
    MaterialAttributes=@n57.MaterialAttributes
    Emissive Color=StaticSwitchParameter(True=@n58, False=@n59) {parameterName=Emissive}
    Normal=Transform(None=@n57.Normal) {transformSourceType=TRANSFORMSOURCE_World, transformType=TRANSFORM_Tangent}
   ) {attributeSetTypes [#7, #5]}
  )
  n60 = VectorParameter() {group=Base, parameterName=TintColor, sortPriority=0, defaultValue=<rgba 0.168 0.1154 0.2594 1>}
  n61 = Constant() {r=3}
  n62 = Constant() {r=0.025}
  n63 = PixelNormalWS()
  n64 = StaticSwitchParameter(
   True=@n60.RGB
   False=LinearInterpolate(
    A=@n60.RGB
    B=VectorParameter() {group=Base, parameterName=TintFresnel, defaultValue=<rgba 0.3802 0.1219 0.2848 0>}.RGB
    Alpha=Fresnel(ExponentIn=@n61, BaseReflectFractionIn=@n62, Normal=@n63) {baseReflectFraction=0, exponent=3}
   )
  ) {group=RVT, parameterName=NoRVT}
  n65 = ShadingPathSwitch(Default=@n64, Deferred=@n64, Forward=@n60.RGB, Mobile=@n60.RGB)
  n66 = TextureCoordinate() {coordinateIndex=2}
  n67 = TextureCoordinate()
  n68 = TextureSampleParameter2D(UVs=Multiply(
    A=StaticSwitchParameter(True=@n66, False=@n67) {group=Base, parameterName=Use UV3 for Spec}
    B=ScalarParameter() {defaultValue=1, group=Base, parameterName=Tile Spec}
   )) {
   group=Base
   parameterName=Spec
   samplerType=SAMPLERTYPE_Masks
   texture=/Game/Environments/Apollo/Sets/LuxurySupreme_CrashSite/Textures/T_LS_Crash_Wreckage_Purple_Surface_S
  }
  n69 = TextureSampleParameter2D(UVs=Multiply(A=TextureCoordinate(), B=ScalarParameter() {defaultValue=1, group=Base, parameterName=Tile Normal Map})) {
   group=Base
   parameterName=Normals
   samplerSource=SSM_Wrap_WorldGroupSettings
   samplerType=SAMPLERTYPE_Normal
   sortPriority=1
   texture=/Engine/EngineMaterials/FlatNormal
  }
  n70 = AppendVector(
   A=Multiply(A=ComponentMask(None=@n69.RGB) {g=true, r=true}, B=ScalarParameter() {defaultValue=1, parameterName=Normals Value})
   B=ComponentMask(None=@n69.RGB) {b=true}
  )
  n71 = SetMaterialAttributes!(MaterialAttributes, Base Color=@n65, Metallic=@n68.G, Specular=@n68.G, Roughness=@n68.G, Normal=@n70) {attributeSetTypes [
    #1
    #2
    #6
    #3
    #5
   ]}
  n72 = SetMaterialAttributes(
   Base Color=@n65
   Metallic=ScalarParameter() {defaultValue=0.5, group=Base, parameterName=Metallic, sortPriority=2}
   Specular=ScalarParameter() {defaultValue=0.5, group=Base, parameterName=Specular, sortPriority=3}
   Roughness=ScalarParameter() {defaultValue=0.3, group=Base, parameterName=Roughness, sortPriority=4}
   Normal=@n70
  ) {attributeSetTypes [
    #1
    #2
    #6
    #3
    #5
   ]}
  n73 = StaticSwitchParameter(True=@n71, False=@n72) {group=Base, parameterName="Use Textures "}
  n74 = GetMaterialAttributes(None=StaticSwitchParameter(True=NamedRerouteUsage(), False=NamedRerouteDeclaration(None=@n73) {name=Textures, nodeColor=<rgba 0 1 0.4844 1>}) {group=RVT, parameterName=NoRVT}) {attributeGetTypes [#5]}
  n75 = TextureObjectParameter() {
   group=Tiling Layer
   parameterName=Details Normals
   samplerType=SAMPLERTYPE_Normal
   sortPriority=3
   texture=/Game/Global/Textures/Default/Tileable_Patterns/T_Skyfire_Blocks_Thin_N
  }
  n76 = TextureSample(
   UVs=Multiply(A=TextureCoordinate(), B=ScalarParameter() {defaultValue=1, group=Tiling Layer, parameterName=Scale Details, sortPriority=1})
   Tex=@n75
  ) {samplerType=SAMPLERTYPE_Normal, texture=&FlatNormal}
  n77 = Constant3Vector() {constant=<rgba 1 1 0 0>}
  n78 = Add(
   A=@n74.Normal
   B=Multiply(A=Multiply(A=@n76.RGB, B=@n77), B=ScalarParameter() {defaultValue=0.25, group=Tiling Layer, parameterName=DetailsValue, sortPriority=2})
  )
  NamedRerouteDeclaration(None=StaticSwitchParameter(
    True=SetMaterialAttributes!(MaterialAttributes=@n74.MaterialAttributes, Normal=@n78) {attributeSetTypes [#5]}
    False=@n74.MaterialAttributes
   ) {group=Tiling Layer, parameterName=Tiling Layer}) {name=Base, nodeColor=<rgba 0.1719 1 0 1>}
  n79 = VectorParameter() {group=Emissive, parameterName=EmissiveColor}
  n80 = Multiply(A=TextureCoordinate(), B=ScalarParameter() {defaultValue=5, group=Emissive, parameterName=Scale Emissive UVs})
  n81 = TextureSampleParameter2D(UVs=@n80) {
   group=Emissive
   parameterName=Emissive
   samplerSource=SSM_Wrap_WorldGroupSettings
   samplerType=SAMPLERTYPE_Masks
   texture=/Game/Global/Textures/Default/Tileable_Patterns/Crashed_Emissive_M
  }
  n82 = Add(A=@n79.RGB, B=Multiply(A=@n79.RGB, B=@n81.R))
  n83 = OneMinus(None=Fresnel(Normal=VertexNormalWS()) {exponent=3})
  n84 = TextureSampleParameter2D(UVs=TextureCoordinate() {coordinateIndex=1}) {
   group=Damage
   parameterName=DamageMask
   samplerSource=SSM_Wrap_WorldGroupSettings
   samplerType=SAMPLERTYPE_LinearGrayscale
   sortPriority=4
   texture=&T_LS_Crash_Wreckage_Mask_01_M
  }
  n85 = Multiply(A=Multiply(A=Multiply(A=@n82, B=@n83), B=@n84.R), B=@n81.G)
  n86 = MaterialFunctionCall(
   Transition Phase=Divide(A=NamedRerouteUsage(), B=ScalarParameter() {defaultValue=500, group=Emissive, parameterName=EmissiveFade})
   Height Texture=OneMinus(None=@n81.G)
  ) {materialFunction=&HeightLerp}
  n87 = ShadingPathSwitch(
   Default=StaticSwitchParameter(True=@n85, False=Multiply(A=@n85, B=@n86.Lerp Alpha No Contrast)) {group=RVT, parameterName=NoRVT}
   Forward=@n85
   Mobile=@n85
  )
  n88 = Multiply(A=Time(), B=ScalarParameter() {defaultValue=0.045, parameterName=Emissive_Flicker_Tiime})
  n89 = ScalarParameter() {defaultValue=0.5, group=HLOD, parameterName=HLODEmissiveTime}
  n90 = TextureSampleParameter2D(UVs=Multiply(A=MaterialProxyReplace(Realtime=@n88, MaterialProxy=@n89), B=Constant() {r=0.2})) {
   group=Emissive
   parameterName=Emissive
   samplerSource=SSM_Wrap_WorldGroupSettings
   samplerType=SAMPLERTYPE_Masks
   texture=&Crashed_Emissive_M
  }
  n91 = ScalarParameter() {defaultValue=1.35, parameterName=Emissive_Mask_Amount}
  n92 = Multiply(
   A=MaterialProxyReplace(Realtime=Multiply(A=@n87, B=@n81.B), MaterialProxy=@n85)
   B=MaterialFunctionCall(A=@n84.R, Transition Phase=Multiply(A=@n90.R, B=@n91), Height Texture=@n84.R, Contrast=Constant() {r=1}) {materialFunction=&HeightLerp}.Alpha
  )
  NamedRerouteDeclaration(None=@n92) {name=HLODEmissive, nodeColor=<rgba 1 0 0.6328 1>}
  n93 = Divide(
   A=NamedRerouteUsage()
   B=ComponentMask(None=MaterialFunctionCall() {materialFunction=&ObjectLocalBounds}.Local Bounds Max) {b=true}
  )
  n94 = Constant()
  n95 = Divide(
   A=NamedRerouteDeclaration(None=ShadingPathSwitch(Default=@n93, Forward=@n94, Mobile=@n94)) {name=MaxZ, nodeColor=<rgba 1 0 0.7266 1>}
   B=ScalarParameter() {defaultValue=0.75, parameterName=DamageHeight}
  )
  n96 = ScalarParameter() {defaultValue=1, parameterName=HLODHeight}
  NamedRerouteDeclaration(None=Saturate(None=MaterialProxyReplace(Realtime=@n95, MaterialProxy=@n96))) {name=GroundMaskV2, nodeColor=<rgba 0 0.7891 1 1>}
  n97 = ScalarParameter() {defaultValue=1200, group=Damage, parameterName=Damage distance, sortPriority=0}
  n98 = Constant() {r=2}
  n99 = VertexNormalWS()
  NamedRerouteDeclaration(None=LinearInterpolate(A=Divide(A=@n97, B=@n98), B=@n97, Alpha=ComponentMask(None=@n99) {b=true})) {name=RVTDamage, nodeColor=<rgba 1 0 0.2344 1>}
  NamedRerouteDeclaration(None=@n28) {name=AngeMask, nodeColor=<rgba 1 0 0.4687 1>}
  NamedRerouteDeclaration(None=StaticSwitchParameter(True=@n32.Alpha, False=@n33) {group=RVT, parameterName=NoRVT}) {name=NoRVT Painted Damage Mask, nodeColor=<rgba 0 0.1562 1 1>}
  n100 = TextureObjectParameter() {group=Damage, parameterName=DamageHeight, samplerType=SAMPLERTYPE_LinearGrayscale, texture=/Game/Global/Textures/Default/Noise/BentNoise_H}
  n101 = TextureObjectParameter() {group=Damage, parameterName=DamageNormals, samplerType=SAMPLERTYPE_Normal, texture=/Game/Global/Textures/Default/Noise/BentNoise_N}
  n102 = MaterialFunctionCall!(
   Diffuse Texture Object=@n100
   Spec Texture Object
   TextureSize=ScalarParameter() {defaultValue=2500, group=Damage, parameterName=Damage Texture Scale, sortPriority=1}
   --------------------------
   Use seperate settings for Z projection
   Use seperate texture for Z projection
   Texture Size Z projection
   Z Plane Diffuse Texture Object
   Z Plane Spec Texture Object
   "--------------------------      "
   ProjectionTransitionContrast
   Use Texture Alpha Heightmap to Refine the Transistion
   HeightLerpTransitionContrast
   PickTheDiffuseHeightChannel
   TheDiffuseHeightChannel
   "--------------------------       "
   NormalTextureObject=NamedRerouteDeclaration(None=@n101) {name=Damage Normal, nodeColor=<rgba 0.7344 1 0 1>}
   World Space Normals=StaticBool()
   UseUncorrectedNormalBlend
   "--------------------------  "
   Export Float 4 Diffuse
   Export Float 4 Spec
   "--------------------------     "
   WorldPosition
   World Space Normal
  ) {materialFunction=/Game/Packages/Fortress_SharedMaterials/Base_Material_Functions/WorldAlignedTextures_Complex_OptimzedNormal}
  n103 = ComponentMask(None=@n102.Normal XYZ Texture) {g=true, r=true}
  n104 = ScalarParameter() {defaultValue=1, group=Damage, parameterName="Damage N Value  Global"}
  NamedRerouteDeclaration(None=AppendVector(A=Multiply(A=@n103, B=@n104), B=ComponentMask(None=@n102.Normal XYZ Texture) {b=true})) {name=Damage N WS, nodeColor=<rgba 0 1 0.2969 1>}
  NamedRerouteDeclaration(None=ComponentMask(None=@n102.XYZ Diffuse Texture) {r=true}) {name=Damage HM WS, nodeColor=<rgba 0.8281 1 0 1>}
  n105 = TextureSample(UVs=@n80) {samplerType=SAMPLERTYPE_Normal, texture=/Game/Global/Textures/Default/Tileable_Patterns/Crashed_Emissive_N}
  n106 = Multiply(A=NamedRerouteUsage(), B=Constant3Vector() {constant=<rgba 1 1 0 0>})
  n107 = SetMaterialAttributes!(
   MaterialAttributes
   Emissive Color=@n92
   Normal=StaticSwitchParameter(True=@n105.RGB, False=Add(A=@n105.RGB, B=@n106)) {group=RVT, parameterName=NoRVT}
   Specular=Constant() {r=0.05}
   Base Color=Constant3Vector() {constant=<rgba 0.077 0.0337 0.0781 0>}
   Metallic=Constant() {r=0.5}
   Roughness=Constant() {r=0.45}
   Custom Data 1=Constant() {r=0.15}
   Custom Data 0=Constant() {r=1}
  ) {attributeSetTypes [
    #7
    #5
    #6
    #1
    #2
    #3
    #8
    #4
   ]}
  NamedRerouteDeclaration(None=@n107) {name=EmissiveBranch, nodeColor=<rgba 0 0.2266 1 1>}
  n108 = SetMaterialAttributes!(
   MaterialAttributes
   Base Color=Constant() {r=0.25}
   Metallic=Constant() {r=1}
   Roughness=Constant() {r=0.2}
   Custom Data 0=Constant()
   Normal=NamedRerouteUsage()
  ) {attributeSetTypes [
    #1
    #2
    #3
    #4
    #5
   ]}
  n109 = TextureObject() {samplerType=SAMPLERTYPE_Grayscale, texture=&LowResBlurredNoise}
  n110 = Constant() {r=5000}
  n111 = BlendMaterialAttributes(
   A=@n108
   B=NamedRerouteUsage()
   Alpha=MaterialFunctionCall(
    Transition Phase=Constant() {r=0.62}
    Height Texture=MaterialFunctionCall(TextureObject=@n109, TextureSize=@n110) {materialFunction=&WorldAlignedTexture}.XYZ Texture
    Contrast=Constant() {r=0.2}
   ) {materialFunction=&HeightLerp}.Alpha
  )
  NamedRerouteDeclaration(None=@n111) {name=TexturesNORVT, nodeColor=<rgba 0 0.2734 1 1>}
  Reroute(None=@n73)
  Reroute(None=@n111)
 ]
}
