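# Unlit backdrop material for the Luxury Supreme crash site, built from Skyfire exterior textures and enabled for skeletal meshes.
# Panning smoke modulates emissive from sharp plus blurred-mip samples, added to diffuse that blends from blurred to sharp by the emissive mask.
# Atmospheric light against the normal-mapped reflection vector fakes tight and broad highlights tinted by SkyTint.
# Static smoke and camera-relative height mask the result, darkening it toward and below camera level.
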Material /Game/Environments/Apollo/Sets/LuxurySupreme_CrashSite/Materials/M_LS_CrashSite_Backdrop_Base_01
graph {
 properties {
  bUseMaterialAttributes=true
  bUsedWithMeshDeformer=true
  bUsedWithSkeletalMesh=true
  pixelDepthOffsetMode=PDOM_Legacy
  shadingModel=MSM_Unlit
  usedShadingModels=Unlit
 }
 nodes [
  n1 = Saturate(None=LinearInterpolate(A=Constant() {r=-0.1}, B=Constant() {r=1}, Alpha=TextureSample() {texture=/Engine/Tutorial/SubEditors/TutorialAssets/T_soft_smoke}.R))
  n2 = WorldPosition() {worldPositionShaderOffset=WPT_CameraRelative}
  n3 = Constant() {r=75000}
  n4 = Saturate(None=ComponentMask(None=Divide(A=@n2.XYZ, B=@n3)) {b=true})
  n5 = Panner(Speed=Constant() {r=-0.1}) {speedX=0.01, speedY=0.02}
  n6 = TextureSampleParameter2D() {parameterName=Emissive, texture=/Game/Gadgets/Assets/Skyfire_Exterior_Gadget/Textures/T_Skyfire_Exterior_E}
  n7 = TextureSampleParameter2D!(UVs, Level=Constant() {r=4}, Apply View MipBias) {mipValueMode=TMVM_MipLevel, parameterName=Emissive, texture=&T_Skyfire_Exterior_E}
  n8 = Constant() {r=7}
  n9 = Add(
   A=Multiply(A=TextureSample(UVs=@n5) {texture=&T_soft_smoke}.R, B=Add(A=@n6.RGB, B=@n7.RGB))
   B=LinearInterpolate(
    A=TextureSample!(UVs, Tex, Level=@n8, Apply View MipBias) {mipValueMode=TMVM_MipLevel, texture=/Game/Gadgets/Assets/Skyfire_Exterior_Gadget/Textures/T_Skyfire_Exterior_D}.RGB
    B=TextureSample!(UVs, Tex, Apply View MipBias) {texture=&T_Skyfire_Exterior_D}.RGB
    Alpha=@n6.R
   )
  )
  n10 = TextureSampleParameter2D!(UVs, Apply View MipBias) {parameterName=Normals, samplerType=SAMPLERTYPE_Normal, texture=/Game/Gadgets/Assets/Skyfire_Exterior_Gadget/Textures/T_Skyfire_Exterior_N}
  n11 = DotProduct(A=AtmosphericLightVector(), B=ReflectionVectorWS(CustomWorldNormal=Transform(None=@n10.RGB)))
  n12 = Constant() {r=45}
  n13 = Multiply(A=Constant() {r=15}, B=Saturate(None=Power(Base=@n11, Exp=@n12)))
  n14 = TextureSampleParameter2D() {parameterName=Emissive, texture=&T_Skyfire_Exterior_E}
  n15 = Saturate(None=Power(Base=@n11, Exp=Constant() {r=5}))
  n16 = Constant() {r=8}
  n17 = Multiply(
   A=VectorParameter() {parameterName=SkyTint, defaultValue=<rgba 0.0307 0.0592 0.1224 0>}.RGB
   B=Add(A=Multiply(A=@n13, B=@n14.R), B=Multiply(A=@n15, B=@n16))
  )
  SetMaterialAttributes(Emissive Color=Multiply(A=Multiply(A=@n1, B=@n4), B=Add(A=@n9, B=@n17))) {attributeSetTypes [#1]}
 ]
}
